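# Sound cue for the Stink Flopper item's explosion, layering a wet fish splat with a stink grenade burst.
# A random splat plays immediately alongside a random close AOE burst that starts after a short fixed delay, each with pitch variation.
# The mix runs through close-explosion attenuation with overrides and a third-person ranged-weapon-fire sound class.
# The root mixer has two empty inputs, and a concurrency override caps simultaneous instances at three.
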SoundCue /Game/Athena/Sounds/Items/Flopper/StinkFlopper/Stink_Flopper_Explode_Cue
root=Mixer_5
nodes [
 {
  id=Attenuation_6
  children [Mixer_2]
  properties {
   attenuationSettings=/Game/Sounds/Attenuation/Fort_Explosion_Close_Athena_Att
   attenuationOverrides {falloffDistance=900, attenuationShapeExtents {x=100}}
  }
 }
 {id=Delay_1, children [Modulator_3], properties {delayMax=0.1, delayMin=0.1}}
 {id=Mixer_2, children [Delay_1,Modulator_4], properties {inputVolume [1,0.75]}}
 {id=Mixer_5, children [SoundClass_3,null,null], properties {inputVolume [0.7,1.2,1.2]}}
 {id=Modulator_3, children [Random_1], properties {pitchMax=1.1, pitchMin=0.9, volumeMax=1, volumeMin=1}}
 {id=Modulator_4, children [Random_3], properties {pitchMax=1.3, pitchMin=1, volumeMax=1, volumeMin=1}}
 {id=Random_1, children [WavePlayer_3, WavePlayer_4, WavePlayer_11], properties {weights [1,1,1]}}
 {id=Random_3, children [WavePlayer_10, WavePlayer_13, WavePlayer_14, WavePlayer_15], properties {weights [1,1,1,1]}}
 {
  id=SoundClass_3
  children [Attenuation_6]
  properties {soundClassOverride=&Fort_WeaponRanged_Fire_3P_PSM}
 }
 {id=WavePlayer_10, properties {soundWaveAssetPtr=&/Stink_Fish_Splats_01}}
 {id=WavePlayer_11, properties {soundWaveAssetPtr=/Game/Athena/Sounds/Weapons/Stink_Grenade/Stink_Grenade_AOE_Burst_Close_03}}
 {id=WavePlayer_13, properties {soundWaveAssetPtr=&/Stink_Fish_Splats_02}}
 {id=WavePlayer_14, properties {soundWaveAssetPtr=&/Stink_Fish_Splats_03}}
 {id=WavePlayer_15, properties {soundWaveAssetPtr=&/Stink_Fish_Splats_04}}
 {id=WavePlayer_3, properties {soundWaveAssetPtr=/Game/Athena/Sounds/Weapons/Stink_Grenade/Stink_Grenade_AOE_Burst_Close_01}}
 {id=WavePlayer_4, properties {soundWaveAssetPtr=/Game/Athena/Sounds/Weapons/Stink_Grenade/Stink_Grenade_AOE_Burst_Close_02}}
]
properties {bOverrideConcurrency=true, maxDistance=1499.9875, volumeMultiplier=1, concurrencyOverrides {maxCount=3}}
