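# Sound cue for the looping in-air sound of the Stink Flopper item.
# Mixes a looping stink fish gas wave with a quieter reused toy tomato in-air loop, then runs the blend through Doppler and a default Modulator.
# Plays in the explosives sound class with close-explosion attenuation and a concurrency override that caps simultaneous instances.
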SoundCue /Game/Athena/Sounds/Items/Flopper/StinkFlopper/Stink_Flopper_InAir_Loop_Cue
root=Modulator_0
nodes [
 {id=Doppler_0, children [Mixer_0], properties {dopplerIntensity=2}}
 {id=Mixer_0, children [WavePlayer_1,WavePlayer_2], properties {inputVolume [1,0.4]}}
 {id=Modulator_0, children [Doppler_0]}
 {id=WavePlayer_1, properties {bLooping=true, soundWaveAssetPtr=&/Stink_Fish_Gas_Loop}}
 {id=WavePlayer_2, properties {bLooping=true, soundWaveAssetPtr=/Game/Athena/Sounds/Toys/Toy_Tomato_InAirLoop}}
]
properties {
 attenuationSettings=/Game/Sounds/Attenuation/Fort_Explosion_Close_StW_Att
 bOverrideConcurrency=true
 maxDistance=3500
 soundClassObject=/Game/Sounds/FortSoundClassesAndModes/Fort_Explosives
 volumeMultiplier=0.4
 attenuationOverrides {distanceAlgorithm=Logarithmic, falloffDistance=1200, attenuationShapeExtents {x=500}}
 concurrencyOverrides {maxCount=2}
}
