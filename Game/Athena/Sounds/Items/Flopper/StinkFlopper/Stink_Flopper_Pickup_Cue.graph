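# Sound cue for picking up the Stink Flopper fish item in Athena.
# It plays the Puffer_Fish_PU wave through a modulator node.
# The cue routes to the loot-collect sound class at reduced volume and uses the shared weapon pickup attenuation for spatial falloff.
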SoundCue /Game/Athena/Sounds/Items/Flopper/StinkFlopper/Stink_Flopper_Pickup_Cue
root=Modulator_1
nodes [
 {id=Modulator_1, children [WavePlayer_1]}
 {id=WavePlayer_1, properties {soundWaveAssetPtr=/Game/Athena/Sounds/Items/Flopper/Puffer_Fish_PU}}
]
properties {
 attenuationSettings=&Fort_Weapon_Pickup_Att
 maxDistance=3500
 soundClassObject=&Fort_Loot_Collect
 volumeMultiplier=0.45
}
