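# Sound cue for the Stink Flopper's stink release, layering looping fish gas sounds with reused Stink Grenade area-of-effect sounds.
# The root mixes two branches with separate attenuations: a smoke-grenade layer of three gas loops and a stink "inside" layer of loops plus repeating random hiss one-shots.
# Envelopers ramp up the whole first branch but only the fish gas loop in the inside branch.
# Modulators vary pitch, and a gas concurrency set governs simultaneous instances.
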SoundCue /Game/Athena/Sounds/Items/Flopper/StinkFlopper/Stink_Flopper_ReleaseStink_Cue
root=Mixer_0
nodes [
 {id=Attenuation_2, children [Enveloper_0], properties {attenuationSettings=/Game/Sounds/Attenuation/Fort_Grenade_Smoke_Att}}
 {
  id=Attenuation_3
  children [Mixer_1]
  properties {attenuationSettings=/Game/Sounds/Attenuation/Fort_Grenade_Smoke_Stink_Inside_Att, attenuationOverrides {bAttenuate=false, bSpatialize=false}}
 }
 {
  id=Enveloper_0
  children [Mixer_2]
  properties {pitchCurve {editorCurveData {keys [{time=0}]}}, volumeCurve {editorCurveData {keys [<at 0.0157 0.0193>,<at 0.5067 0.3866>,<at 0.9456 1>]}}}
 }
 {
  id=Enveloper_1
  children [Modulator_2]
  properties {pitchCurve {editorCurveData {keys [{time=0}]}}, volumeCurve {editorCurveData {keys [<at 0.0157 0.0193>,<at 0.5067 0.3866>,<at 0.9456 1>]}}}
 }
 {id=Looping_0, children [Random_0]}
 {id=Mixer_0, children [Attenuation_2, Attenuation_3], properties {inputVolume [1,1.25]}}
 {id=Mixer_1, children [Modulator_3,Modulator_1,Enveloper_1], properties {inputVolume [0.75,1.5,0.75]}}
 {id=Mixer_2, children [Modulator_4,Modulator_0,Modulator_5], properties {inputVolume [0.5,0.75,0.15]}}
 {id=Modulator_0, children [WavePlayer_4], properties {pitchMax=1.1, pitchMin=0.9, volumeMax=1, volumeMin=1}}
 {id=Modulator_1, children [Looping_0], properties {pitchMax=1.1, pitchMin=0.9, volumeMax=1, volumeMin=0.8}}
 {id=Modulator_2, children [WavePlayer_0], properties {pitchMax=1.1, pitchMin=0.9, volumeMax=1, volumeMin=1}}
 {id=Modulator_3, children [WavePlayer_5], properties {pitchMax=1.1, pitchMin=0.9, volumeMax=1, volumeMin=1}}
 {id=Modulator_4, children [WavePlayer_1], properties {pitchMax=1.1, pitchMin=0.9, volumeMax=1, volumeMin=1}}
 {id=Modulator_5, children [WavePlayer_2], properties {pitchMax=1.3, pitchMin=0.98}}
 {
  id=Random_0
  children [
   WavePlayer_6
   WavePlayer_7
   WavePlayer_8
   WavePlayer_9
   WavePlayer_10
   WavePlayer_11
   WavePlayer_12
   WavePlayer_13
   WavePlayer_14
   WavePlayer_15
  ]
  properties {weights [1,1,1,1,1,1,1,1,1,1]}
 }
 {id=WavePlayer_0, properties {bLooping=true, soundWaveAssetPtr=&/Stink_Fish_Gas_Loop}}
 {id=WavePlayer_1, properties {bLooping=true, soundWaveAssetPtr=&Stink_Fish_Gas_Loop}}
 {id=WavePlayer_10, properties {soundWaveAssetPtr=/Game/Athena/Sounds/Weapons/Stink_Grenade/Stink_Grenade_AOE_Close_Inside_Hiss_04}}
 {id=WavePlayer_11, properties {soundWaveAssetPtr=/Game/Athena/Sounds/Weapons/Stink_Grenade/Stink_Grenade_AOE_Close_Inside_Hiss_05}}
 {id=WavePlayer_12, properties {soundWaveAssetPtr=/Game/Athena/Sounds/Weapons/Stink_Grenade/Stink_Grenade_AOE_Close_Inside_Hiss_06}}
 {id=WavePlayer_13, properties {soundWaveAssetPtr=/Game/Athena/Sounds/Weapons/Stink_Grenade/Stink_Grenade_AOE_Close_Inside_Hiss_07}}
 {id=WavePlayer_14, properties {soundWaveAssetPtr=/Game/Athena/Sounds/Weapons/Stink_Grenade/Stink_Grenade_AOE_Close_Inside_Hiss_08}}
 {id=WavePlayer_15, properties {soundWaveAssetPtr=/Game/Athena/Sounds/Weapons/Stink_Grenade/Stink_Grenade_AOE_Close_Inside_Hiss_09}}
 {id=WavePlayer_2, properties {bLooping=true, soundWaveAssetPtr=&/Stink_Fish_High_Gas_Loop}}
 {id=WavePlayer_4, properties {bLooping=true, soundWaveAssetPtr=/Game/Athena/Sounds/Weapons/Stink_Grenade/Stink_Grenade_AOE_Close_Loop}}
 {id=WavePlayer_5, properties {bLooping=true, soundWaveAssetPtr=/Game/Athena/Sounds/Weapons/Stink_Grenade/Stink_Grenade_AOE_Close_Inside_Loop}}
 {id=WavePlayer_6, properties {soundWaveAssetPtr=/Game/Athena/Sounds/Weapons/Stink_Grenade/Stink_Grenade_AOE_Close_Inside_Hiss_10}}
 {id=WavePlayer_7, properties {soundWaveAssetPtr=/Game/Athena/Sounds/Weapons/Stink_Grenade/Stink_Grenade_AOE_Close_Inside_Hiss_01}}
 {id=WavePlayer_8, properties {soundWaveAssetPtr=/Game/Athena/Sounds/Weapons/Stink_Grenade/Stink_Grenade_AOE_Close_Inside_Hiss_02}}
 {id=WavePlayer_9, properties {soundWaveAssetPtr=/Game/Athena/Sounds/Weapons/Stink_Grenade/Stink_Grenade_AOE_Close_Inside_Hiss_03}}
]
properties {
 maxDistance=6800
 soundClassObject=&Fort_WeaponMelee_PSM
 volumeMultiplier=1
 concurrencyOverrides {maxCount=2, volumeScale=0.7}
 concurrencySet [&/Stink_Flopper_Gas_Conc]
}
