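# Sound cue for throwing the Stink Flopper fish item in Athena, attenuated with weapon-drop settings under the loot-collect sound class.
# A mixer layers a quieter Rift fish pickup sound with two delayed branches, all under a root modulator.
# One branch pairs a random run foley with a quieter Molotov throw after a fixed delay.
# The other plays a fish splat after a short variable delay, with an enveloper fading its volume in.
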SoundCue /Game/Athena/Sounds/Items/Flopper/StinkFlopper/Stink_Flopper_Throw_Cue
root=Modulator_0
nodes [
 {id=Delay_0, children [Enveloper_0], properties {delayMax=0.1, delayMin=0.05}}
 {id=Delay_1, children [Mixer_1], properties {delayMax=0.08, delayMin=0.08}}
 {
  id=Enveloper_0
  children [WavePlayer_0]
  properties {pitchCurve {editorCurveData {keys [{time=0}]}}, volumeCurve {editorCurveData {keys [<at 0.0191 0.033>,<at 0.3732 1>]}}}
 }
 {id=Mixer_0, children [WavePlayer_4,Delay_1,Delay_0], properties {inputVolume [0.25,1,1]}}
 {id=Mixer_1, children [Random_0,WavePlayer_5], properties {inputVolume [1,0.5]}}
 {id=Modulator_0, children [Mixer_0]}
 {id=Random_0, children [WavePlayer_1,WavePlayer_2,WavePlayer_3], properties {weights [1,1,1]}}
 {id=WavePlayer_0, properties {soundWaveAssetPtr=&/Stink_Fish_Splats_01}}
 {id=WavePlayer_1, properties {soundWaveAssetPtr=/Game/Sounds/Fort_Foley_Sounds/Moves/player_run_foley_11}}
 {id=WavePlayer_2, properties {soundWaveAssetPtr=/Game/Sounds/Fort_Foley_Sounds/Moves/player_run_foley_12}}
 {id=WavePlayer_3, properties {soundWaveAssetPtr=/Game/Sounds/Fort_Foley_Sounds/Moves/player_run_foley_02}}
 {id=WavePlayer_4, properties {soundWaveAssetPtr=/Game/Athena/Sounds/Items/Flopper/Rift_Fish_PU}}
 {id=WavePlayer_5, properties {soundWaveAssetPtr=/Game/Sounds/Fort_Weapons/Grenades/FireJar/Molotov_Throw}}
]
properties {
 attenuationSettings=&Fort_Weapon_Drop_Att
 maxDistance=3000
 soundClassObject=&Fort_Loot_Collect
 volumeMultiplier=0.65
}
